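; War Effort outcome handler component, attached to a funding station actor to act on funded outcomes.
; On BeginPlay it casts its owner to B_WarEffort_FundingStation_Base and stores the result as FundingStationOwner.
; DoFundedOutcome takes an outcome data object and an initial-fund flag but has an empty body here, as does Tick.
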(event EventBeginPlay
  (bind _fundingstation (CallFunction|GetOwnerFundingStation))
  (Variables|Default|SetFundingStationOwner _fundingstation))

(event EventTick
  (params (DeltaSeconds :type Float%20%28single-precision%29)))

(fn DoFundedOutcome
  (params (OutcomeData :type War%20Effort%20Outcome%20Data%20Object%20Reference) (IsInitialFund :type Boolean)))

(fn GetOwnerFundingStation
  (bind _asb_war_effort_funding_station_base (Utilities|Casting|CastToB_WarEffort_FundingStation_Base (Components|GetOwner)))
  (return _asb_war_effort_funding_station_base))
