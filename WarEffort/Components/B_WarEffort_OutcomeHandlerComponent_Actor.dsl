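; War Effort outcome handler component that spawns environmental actors when an outcome is funded.
; DoFundedOutcome casts the outcome data to the environmental actor type and reads its actor class.
; It then has each funding outcome proxy actor spawn that class on the ground, without using the override as an offset.
; IsInitialFund is unused, and a debug print reports "Spawning turrets" whatever the class.
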(fn DoFundedOutcome
  (params (OutcomeData :type War%20Effort%20Outcome%20Data%20Object%20Reference) (IsInitialFund :type Boolean))
  (bind _aswar_effort_outcome_data_environmental_actor (Utilities|Casting|CastToWarEffortOutcomeData_EnvironmentalActor OutcomeData))
  (LocalVariables|SetActorClass (Variables|/WarEffort/Data/OutcomeData/WarEffortOutcomeData_EnvironmentalActor.WarEffortOutcomeData_EnvironmentalActor_C|GetEnvironmentalActorClass _aswar_effort_outcome_data_environmental_actor))
  (Development|PrintString "Spawning turrets")
  (for _array_element (Variables|Default|GetProxyActors)
    (Class|BWarEffortFundingOutcomeProxy|SpawnActoratProxy :self _array_element :ActorClass (LocalVariables|GetActorClass) :SpawnOnGround true :UseOverrideAsOffset false)))
