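; War Effort outcome handler component that spawns item pickups at its funding station when an outcome is first funded.
; DoFundedOutcome reads loot outcome data, preferring a loot tier name and falling back to a single item definition when the name is empty.
; A scalable float sets the spawn count, and each drop gets a random positive offset from the station's spawn point.
; Two custom events perform the tossed spawns, by tier table or item definition.
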(event Custom|SpawnLootByItemDef
  (params (World%20Item%20Definition :type Fort%20World%20Item%20Definition%20Object%20Reference) (Position :type Vector))
  (Inventory|SpawnItemVariantPickupinWorld (Utilities|Struct|MakeSpawnItemVariantParams World%20Item%20Definition "(GameplayTags=)" 1 Position "0.000000,0.000000,1.000000" 0 true true true -1 "Tossed" "ItemSpawner" 0 0 false)))

(event Custom|SpawnLootByTierName
  (params (TierGroupName :type Name) (Position :type Vector))
  (Class|AthenaFunctionLibrary|SpawnLootfromTable TierGroupName 1 -1 Position "0, 0, 0" 0 true true true 0 "Tossed"))

(fn DoFundedOutcome
  (params (OutcomeData :type War%20Effort%20Outcome%20Data%20Object%20Reference) (IsInitialFund :type Boolean))
  (bind _spawnposition (LocalVariables|GetSpawnPosition))
  (if IsInitialFund
    (bind _aswar_effort_outcome_data_loot (Utilities|Casting|CastToWarEffortOutcomeData_Loot OutcomeData))
    (bind _output_get (LocalVariables|SetUseLootTier (!= (Variables|/WarEffort/Data/OutcomeData/WarEffortOutcomeData_Loot.WarEffortOutcomeData_Loot_C|GetLootTierName _aswar_effort_outcome_data_loot) "")))
    (if _output_get
      (LocalVariables|SetLootTierName (Variables|/WarEffort/Data/OutcomeData/WarEffortOutcomeData_Loot.WarEffortOutcomeData_Loot_C|GetLootTierName _aswar_effort_outcome_data_loot))
      (else
        (bind _asfort_world_item_definition (Utilities|Casting|CastToFortWorldItemDefinition (Variables|/WarEffort/Data/OutcomeData/WarEffortOutcomeData_Loot.WarEffortOutcomeData_Loot_C|GetTestItemDefinition _aswar_effort_outcome_data_loot)))
        (LocalVariables|SetItemDef _asfort_world_item_definition)))
    (for _ (range (CallFunction|GetNumItemstoSpawn))
      (LocalVariables|SetSpawnPosition (+ (Getters|GetItemSpawnPoint (Variables|Default|GetFundingStationOwner)) (CallFunction|GetItemSpawnLocationOffset)))
      (if (LocalVariables|GetUseLootTier)
        (CallFunction|SpawnLootbyTierName (LocalVariables|GetLootTierName) _spawnposition)
        (else
          (CallFunction|SpawnLootbyItemDef (LocalVariables|GetItemDef) _spawnposition))))))

(fn GetItemSpawnLocationOffset
  (bind (_locationoffsetmax_x _locationoffsetmax_y _locationoffsetmax_z) (Math|Vector|BreakVector (Variables|GetLocationOffsetMax_X)))
  (return (Math|Vector|MakeVector (Math|Random|RandomFloatinRange 0.0 _locationoffsetmax_x) (Math|Random|RandomFloatinRange 0.0 _locationoffsetmax_y) (Math|Random|RandomFloatinRange 0.0 _locationoffsetmax_z))))

(fn GetNumItemsToSpawn
  (return (Math|Float|Truncate (ScalableFloat|GetValueatLevel (Variables|Default|GetRow_NumItemsToSpawn)))))
