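; War effort outcome handler component for the map change outcome type, overriding the handler's funded-outcome function.
; DoFundedOutcome casts the incoming outcome data to the MapChange data type, then calls the parent implementation with the original data.
; The cast result and the IsInitialFund flag are not used, and no map-changing logic appears in this function.
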(fn DoFundedOutcome
  (params (OutcomeData :type War%20Effort%20Outcome%20Data%20Object%20Reference) (IsInitialFund :type Boolean))
  (Utilities|Casting|CastToWarEffortOutcomeData_MapChange OutcomeData)
  (|CallParentFunction OutcomeData))
